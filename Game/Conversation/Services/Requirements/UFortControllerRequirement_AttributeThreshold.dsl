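; Conversation service requirement that passes when the player's gameplay attribute exceeds a threshold, which the healthCostTable name suggests is a health cost.
; It gets the ability system component from the test context and reads the base value, not the current value, of the configurable attribute.
; The threshold comes from a curve table row at its default input.
; The check is strictly greater-than, and a missing attribute or row never passes.
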(fn IsRequirementMetInternal
  (params (RequestContext :type Controller%20Requirement%20Test%20Context%20Structure%20%28by%20ref%29))
  (bind (_curvetable _rowname) (Utilities|Struct|BreakCurveTableRowHandle (Variables|Default|GethealthCostTable)))
  (bind _gameplayattribute (Variables|Default|GetgameplayAttribute))
  (bind _output_get (LocalVariables|SetplayerAbilitySystem (Ability|GetAbilitySystemComponent (Utilities|Struct|BreakControllerRequirementTestContext RequestContext))))
  (if (Ability|Attribute|GetFloatAttributeBasefromAbilitySystemComponent _output_get _gameplayattribute)
    (bind _outxy (DataTable|EvaluateCurveTableRow _curvetable _rowname)
      (:RowFound
        (if (> (.ReturnValue (Ability|Attribute|GetFloatAttributeBasefromAbilitySystemComponent _output_get _gameplayattribute)) _outxy)
          (return true)
          (else
            _)))
      (:RowNotFound
        _))
    (return false)
    (else
      _)))
